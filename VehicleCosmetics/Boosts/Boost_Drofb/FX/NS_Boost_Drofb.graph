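# Niagara system for the Drofb vehicle boost cosmetic, drawing trail ribbons and exhaust cones at the car's boost sockets with painted color variant overrides.
# NE_Boost spawns ribbons per port with three renderers, tapering width over life and switching backward velocity for frontend preview.
# NE_Cone bursts short-lived mesh and sprite particles in two spawn groups on short loops, partitioned across skeletal mesh sockets, growing and fading as they age.
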NiagaraSystem /VehicleCosmetics/Boosts/Boost_Drofb/FX/NS_Boost_Drofb
dependencies {
 usedDataInterfaces [
  &NiagaraDataInterfaceVectorCurve
  &NiagaraDataInterfaceCurve
  &NiagaraDataInterfaceArrayFloat3
  &NiagaraDataInterfaceSkeletalMesh
 ]
 usedDynamicInputs []
 usedModules [
  &self:SystemSpawnScript.SystemScriptSource.SystemScriptGraph.NiagaraNodeAssignment_0.SetVariables_#1
  &SystemState
 ]
}
emitters {
 NE_Boost {
  emitterUpdateScript [
   &EmitterState(Inactive Response=::NewEnumerator1, Life Cycle Mode=::NewEnumerator1, Loop Duration=1, UseLoopDelay=false)
   &VC_FilteredSocketCount()
   &VC_RateQualityHandler(High Spawn Rate=60, Low Spawn Rate=60, Medium Spawn Rate=60, Use Custom Rate?=true)
   &VC_SpawnRibbon_v1(Number of Ribbons=$Emitter.PortCount, SpawnRate=$Emitter.Spawn Rate)
  ]
  particleSpawnScript [
   &VC_PaintedVariantsOverride(
    DefaultColor=<rgba 0 0.01 0.1 1>
    R2 DefaultColor:LinearColor=<rgba 0 0.01 0.1 1>
    R3 DefaultColor:LinearColor=<rgba 0.2656 0 0 1>
    Use R1 Painted Overrides=true
    Use R2 Painted Overrides=true
    Use R3 Painted Overrides=true
   )
   &VC_InitializeRibbon_v1(
    Color=$Particles.Color
    Custom Offset=true
    Enable Skeletal Position=true
    Lifetime=1.4
    Local Vector Offset=&AddVectorToPosition(Position=$User.DTCarOffset, Vector=<-7 0 0>)
    Offset Amount=1
    Ribbon Width=47
   )
   &AddVelocity(Velocity=&SelectVectorFromArray(
     Direct Array Index=&MakeCustomIntFromBool(Bool=$User.bIsFrontendPreview, True INT=1, False INT=0)
     Vector Selection Array=DataInterface() {propertyValues {FloatData [<-1200 0 0>, <-1400 1 1>]}}
    )) {module=AddVelocity001}
  ]
  particleUpdateScript [
   &ParticleState()
   &InheritSourceMovement(ApplyForce=false, ApplyPosition=true, ApplyPositionFalloff=false)
   &SolveForcesAndVelocity(Clamp Velocity=false, Limit Acceleration=false)
   &VC_UpdateRibbon_v1(Enable Skeletal Position=true)
   &ScaleRibbonWidth(Ribbon Width Scale=&FloatFromCurve(FloatCurve=DataInterface() {propertyValues {Curve {
        defaultValue=1
        keys [{interpMode=RCIM_Cubic, tangentMode=RCTM_User, time=0.05, value=1}, {interpMode=RCIM_Cubic, tangentMode=RCTM_User, time=1, value=0.2}]
       }}}))
   /SetVariables(Particles.whitewidth:NiagaraFloat=&Multiply_Float(
     A=34
     B=&FloatFromCurve(FloatCurve=DataInterface() {propertyValues {Curve {
         defaultValue=1
         keys [
          {arriveTangent=-0.0453, interpMode=RCIM_Cubic, leaveTangent=-0.0453, tangentMode=RCTM_User, time=0, value=1}
          {interpMode=RCIM_Cubic, tangentMode=RCTM_User, time=1, value=0.25}
         ]
        }}})
    )) {isSetParameters=true}
  ]
  rendererClasses [&NiagaraRibbonRendererProperties]
  renderers [
   {rendererClass=&NiagaraRibbonRendererProperties}
   {rendererClass=&NiagaraRibbonRendererProperties, rendererIndex=1}
   {rendererClass=&NiagaraRibbonRendererProperties, rendererIndex=2}
  ]
 }
 NE_Cone {
  emitterUpdateScript [
   &EmitterState(Inactive Response=::NewEnumerator1, Life Cycle Mode=::NewEnumerator1, Loop Duration=0.1, UseLoopDelay=false)
   &VC_FilteredSocketCount()
   &SpawnBurst_Instantaneous(Spawn Count=6, Use Spawn Probability=false, UseLoopCountLimit=false)
   &SpawnBurst_Instantaneous(Spawn Count=9, Spawn Group=1, Use Spawn Probability=false, UseLoopCountLimit=false) {module=SpawnBurst_Instantaneous001}
   &SpawnRate(SpawnRate=30, Use Spawn Probability=false) {enabled=false}
  ]
  particleSpawnScript [
   &VC_PaintedVariantsOverride(
    R1 DefaultColor=<rgba 0.0002 0 0.151 1.5>
    R2 DefaultColor:LinearColor=<rgba 1 1 1 0.4>
    Saffron=<rgba 0.9826 0.9913 0 1>
    Use R1 Painted Overrides=true
    Use R2 Painted Overrides=true
   )
   &InitializeParticle(
    Color Mode=::NewEnumerator0
    Lifetime Max=0.15
    Lifetime Min=0.1
    Lifetime Mode=::NewEnumerator1
    Mesh Scale Mode=::NewEnumerator1
    Mesh Uniform Scale Max=2.5
    Mesh Uniform Scale Min=2
    Position Offset:NiagaraPosition=$User.DTCarOffset
    Position Offset Coordinate Space=::Local
    Sprite Size Mode=::NewEnumerator3
    Uniform Sprite Size=35
    UsePositionOffset=true
    Write Material Random=true
   )
   &PartitionParticles(Number Of Partitions=$Emitter.PortCount)
   &SkeletalMeshLocation(
    Bone / Socket Index=$Particles.PartitionParticles.Partition
    Bone / Socket Sampling Mode=::NewEnumerator5
    MaskSpawnGroup=false
    Mesh Sampling Type=::NewEnumerator6
    Orientation Sampling=::NewEnumerator1
    Sampled Position Offset=&MergeHelper_VectorSwitch(0=<-45 0 0>, 1=<-46 0 0>, 2=?, 3=?, Index=&SpawnGroup(), Num=::NewEnumerator0)
    Skeletal Mesh=$System.SkeletalMesh
   )
   &AddVelocity(Velocity=&RandomRangeVector(Minimum=<5 0 0>, Maximum=<20 0 0>), Velocity Speed Scale=-1) {enabled=false}
   &InitialMeshOrientation(
    Mesh Axis To Orient=<0 -1 0>
    Mesh Orientation Mode=::NewEnumerator2
    Orientation Vector=<-1 0 0>
    Rotation=&RandomRangeVector(Minimum=<0 0 0>, Maximum=<0 1 0>)
    Use Rotation Vector=true
   )
   /SetVariables(Particles.VisibilityTag=$Emitter.SpawnGroup) {isSetParameters=true}
  ]
  particleUpdateScript [
   &ParticleState()
   &SolveForcesAndVelocity(Clamp Velocity=false, Limit Acceleration=false) {enabled=false}
   &ScaleMeshSize(Scale Factor=&VectorFromCurve(
     VectorCurve=DataInterface() {propertyValues {XCurve {keys [<at 0 0.7>,<at 1 1.2>]}, YCurve {keys [<at 0 0.6>,<at 1 1.6>]}, ZCurve {keys [<at 0 0.7>,<at 1 1.2>]}}}
     Scale Curve=<1 0.7 1>
    ))
   &ScaleColor(
    Scale Alpha=&FloatFromCurve(FloatCurve=DataInterface() {propertyValues {Curve {
        defaultValue=1
        keys [
         {interpMode=RCIM_Cubic, tangentMode=RCTM_User, time=0, value=1}
         {arriveTangent=-2.6416, interpMode=RCIM_Cubic, leaveTangent=-2.6416, tangentMode=RCTM_User, time=1, value=0}
        ]
       }}})
    Scale RGB=&VectorFromFloat(Value=1)
    ScaleA=true
    ScaleRGB=false
   )
   /SetVariables(Particles.conescale:Vector3f=&Multiply_Vector(
     A=&VectorFromFloat(Value=&RandomRangeFloat(Minimum=1.9, Maximum=2.3))
     B=&VectorFromCurve(
      VectorCurve=DataInterface() {propertyValues {XCurve {keys [<at 0 0.7>,<at 1 1.2>]}, YCurve {keys [<at 0 0.6>,<at 1 1.6>]}, ZCurve {keys [<at 0 0.7>,<at 1 1.2>]}}}
      Scale Curve=<0.95 0.8 0.95>
     )
    )) {isSetParameters=true}
  ]
  rendererClasses [&NiagaraMeshRendererProperties, &NiagaraSpriteRendererProperties]
  renderers [
   {rendererClass=&NiagaraMeshRendererProperties, rendererIndex=0}
   {rendererClass=&NiagaraMeshRendererProperties, rendererIndex=1}
   {rendererClass=&NiagaraSpriteRendererProperties, rendererIndex=2}
  ]
 }
}
summary {
 systemName=NS_Boost_Drofb
 emitters [
  {emitterName=NE_Cone, rendererClasses [&NiagaraMeshRendererProperties, &NiagaraSpriteRendererProperties]}
  {emitterName=NE_Boost, rendererClasses [&NiagaraRibbonRendererProperties]}
 ]
 userVariables [
  User.bIsFrontendPreview {type=NiagaraBool}
  User.bIsPainted {type=NiagaraBool}
  User.ColorVariantID {type=NiagaraInt32}
  User.SizeScaling {type=NiagaraFloat, defaultValue=1.5, description=""}
  User.PaintedColor {type=LinearColor, description="", defaultValue {struct=LinearColor, value=<rgba 1 1 1 1>}}
  User.PaintedScalar {type=NiagaraFloat, defaultValue=1, description=""}
  User.PortCount {type=NiagaraInt32, defaultValue=2}
  User.DTCarOffset {type=NiagaraPosition, description="", defaultValue {}}
 ]
}
