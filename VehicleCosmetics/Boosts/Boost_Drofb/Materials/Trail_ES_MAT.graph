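# Unlit, translucent, two-sided Niagara ribbon material for the Drofb boost trail.
# Panning distortion noise and caustics warp UVs into a band mask, shaped by circular cutouts and Dynamic Parameter controls for hardness, solidity, add/multiply blend and brightness.
# Particle color is hue-shifted and desaturated, with optional offsets driven by its red and green channels, then feeds emissive adjustments.
# Static switches add distortion and layered sine/cosine World Position Offset waves that fade in along U.
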Material /VehicleCosmetics/Boosts/Boost_Drofb/Materials/Trail_ES_MAT
graph {
 properties {
  bUsedWithNiagaraRibbons=true
  blendMode=BLEND_Translucent
  pixelDepthOffsetMode=PDOM_Legacy
  shadingModel=MSM_Unlit
  twoSided=true
  usedShadingModels=Unlit
 }
 nodes [
  LinearInterpolate()
  LinearInterpolate()
  n1 = ScalarParameter() {parameterName=HueShift}
  n2 = Multiply(A=Constant() {r=0.1}, B=@n1)
  n3 = ParticleColor()
  n4 = Add(A=@n3.G, B=@n3.G)
  n5 = Saturate(None=Add(A=LinearInterpolate(A=@n1, B=@n2, Alpha=@n3.R), B=LinearInterpolate(A=@n2, B=@n1, Alpha=@n4)))
  n6 = Constant() {r=0.5}
  n7 = Constant() {r=0.5}
  n8 = Panner(Coordinate=Multiply(A=Constant2Vector() {g=1, r=6}, B=TextureCoordinate() {vTiling=0.125}), Time=@n7) {speedY=0.5}
  n9 = Constant() {r=3}
  n10 = TextureCoordinate() {uTiling=0.25, vTiling=0.25}
  n11 = TextureSample(UVs=Panner(Coordinate=Multiply(A=@n9, B=@n10)) {speedX=-0.125, speedY=0.01}) {texture=/DealershipSharedLibrary/FX/Textures/Noise/T_VCVFX_Flow02_N}
  n12 = Multiply(A=Constant() {r=0.05}, B=AppendVector(A=@n11.R, B=@n11.R))
  n13 = Multiply(A=Time(), B=Constant() {r=3})
  n14 = TextureSample(UVs=Panner(Coordinate=Add(A=@n8, B=@n12), Time=@n13) {speedX=-1}) {texture=/DealershipSharedLibrary/FX/Textures/Noise/T_VCVFX_DistortNoise02_RGB}
  n15 = Panner(Coordinate=Multiply(A=Constant2Vector() {g=1, r=3}), Time=@n7) {speedX=-1}
  n16 = TextureSample(UVs=Panner(Coordinate=Multiply(A=@n15, B=@n12), Time=@n13) {speedX=-1}) {texture=&T_VCVFX_DistortNoise02_RGB}
  n17 = TextureCoordinate() {uTiling=1.25, vTiling=0.5}
  n18 = Constant() {r=0.5}
  n19 = Multiply(A=Time(), B=ScalarParameter() {defaultValue=2, parameterName=PanSpeed})
  n20 = TextureSample(UVs=Panner(Coordinate=Panner(Coordinate=@n17, Time=@n18) {speedX=0.25, speedY=-0.5}, Time=@n19) {speedY=0.5}) {texture=/DealershipSharedLibrary/FX/Textures/Noise/T_VCVFX_CausticsMaskSharp02_RGB}
  n21 = TextureCoordinate() {vTiling=-0.5}
  n22 = TextureSample(UVs=Panner(Coordinate=Panner(Coordinate=@n21, Time=@n18) {speedY=-0.5}, Time=@n19) {speedX=-0.5, speedY=0.1}) {texture=&T_VCVFX_CausticsMaskSharp02_RGB}
  n23 = LinearInterpolate(A=ComponentMask(None=Add(A=@n14.RGBA, B=@n16.RGBA)) {g=true, r=true}, B=AppendVector(A=@n20.B, B=@n22.B), Alpha=Constant() {r=0.15})
  n24 = ComponentMask(None=Add(A=TextureCoordinate(), B=Multiply(A=@n6, B=@n23))) {g=true}
  n25 = ConstantBiasScale(None=@n24) {bias=-0.13, scale=1}
  n26 = ConstantBiasScale(None=OneMinus(None=@n24)) {bias=0.035, scale=1}
  n27 = Multiply(A=Constant() {r=20}, B=Multiply(A=Saturate(None=@n25), B=Saturate(None=@n26)))
  n28 = DynamicParameter() {defaultValue=<rgba 0.5 1 1 1>, paramNames [Hardness,SolidMask,AddMultiply,Brightness]}
  n29 = Multiply(A=Saturate(None=Multiply(A=@n27, B=@n28.Hardness)), B=@n28.Brightness)
  n30 = LinearInterpolate(A=@n29, B=Saturate(None=@n27), Alpha=@n28.Hardness)
  n31 = ParticleColor()
  n32 = MaterialFunctionCall(
   Hue Shift Percentage=Multiply(A=StaticSwitchParameter(True=@n5, False=@n1) {defaultValue=true, parameterName=UseHueOffset}, B=Saturate(None=@n30))
   Texture=@n31.RGBA
  ) {materialFunction=&HueShift}
  n33 = ScalarParameter() {parameterName=Saturation}
  n34 = Multiply(A=Constant() {r=0.5}, B=@n33)
  n35 = ParticleColor()
  n36 = Add(A=@n35.G, B=@n35.G)
  n37 = Saturate(None=Add(A=LinearInterpolate(A=@n33, B=@n34, Alpha=@n35.R), B=LinearInterpolate(A=@n34, B=@n33, Alpha=@n36)))
  n38 = LinearInterpolate(
   A=Desaturation(None=@n32.Result, Fraction=StaticSwitchParameter(True=@n37, False=@n33) {defaultValue=true, parameterName=UseSaturationOffset})
   B=@n32.Result
  )
  n39 = Saturate(None=@n30)
  MaterialFunctionCall(Emissive=Multiply(A=Constant() {r=2.5}, B=Multiply(A=@n38, B=@n39))) {materialFunction=&MF_FX_EmissiveAdjustments}
  n40 = StaticComponentMaskParameter(None=@n20.RGBA) {defaultG=true, parameterName=MaskRGB_01}
  n41 = StaticComponentMaskParameter(None=@n22.RGBA) {defaultB=true, parameterName=MaskRGB_02}
  n42 = LinearInterpolate(A=Multiply(A=Multiply(A=@n40, B=@n41), B=@n28.Brightness), B=Multiply(A=@n28.Brightness, B=Add(A=@n40, B=@n41)), Alpha=@n28.AddMultiply)
  n43 = Multiply(A=Saturate(None=Multiply(A=@n30, B=@n42)), B=@n28.SolidMask)
  n44 = Multiply(A=@n31.A, B=LinearInterpolate(A=@n29, B=Saturate(None=@n43), Alpha=@n28.SolidMask))
  n45 = Add(A=TextureCoordinate() {uTiling=2, vTiling=0.75}, B=Constant2Vector() {g=-0.385, r=-0.48})
  n46 = OneMinus(None=DotProduct(A=@n45, B=@n45))
  n47 = Constant() {r=0.785}
  n48 = Saturate(None=Multiply(A=Subtract(A=@n46, B=@n47), B=Constant() {r=100}))
  n49 = Add(A=TextureCoordinate() {uTiling=0.65, vTiling=0.75}, B=Constant2Vector() {g=-0.385, r=-0.48})
  n50 = OneMinus(None=DotProduct(A=@n49, B=@n49))
  n51 = Constant() {r=0.785}
  n52 = Saturate(None=Multiply(A=Subtract(A=@n50, B=@n51), B=Constant() {r=100}))
  n53 = Saturate(None=Ceil(None=Add(A=@n48, B=@n52)))
  n54 = Multiply(A=Multiply(A=@n27, B=@n42), B=Constant() {r=30})
  n55 = ComponentMask(None=TextureCoordinate()) {r=true}
  n56 = Saturate(None=ConstantBiasScale(None=@n55) {bias=-0.01, scale=8})
  n57 = Multiply(A=@n56, B=Saturate(None=OneMinus(None=@n55)))
  Multiply(
   A=Saturate(None=Multiply(A=@n44, B=@n53))
   B=StaticSwitchParameter(True=Multiply(A=@n54, B=@n57), False=Constant()) {parameterName=Distortion}
  )
  ParticleColor()
  n58 = Multiply(A=ScalarParameter() {group=WPO, parameterName=Speed}, B=OneMinus(None=Time()))
  n59 = ComponentMask(None=Multiply(A=TextureCoordinate() {coordinateIndex=1, uTiling=3}, B=ScalarParameter() {group=WPO, parameterName=tile})) {r=true}
  n60 = Add(A=@n58, B=@n59)
  n61 = Multiply(A=@n58, B=Constant() {r=0.5})
  n62 = Add(A=Add(A=Cosine(None=@n60), B=Cosine(None=@n60) {period=2}), B=Cosine(None=Add(A=@n59, B=@n61)) {period=0.75})
  n63 = Add(A=@n59, B=@n58)
  n64 = Sine(None=@n63)
  n65 = Sine(None=@n63) {period=2}
  n66 = AppendVector(A=@n62, B=Add(A=Add(A=@n64, B=@n65), B=Sine(None=@n61) {period=0.75}))
  n67 = Multiply(
   A=Multiply(A=AppendVector(A=@n66, B=@n62), B=ScalarParameter() {group=WPO, parameterName=Amount})
   B=ScalarParameter() {group=WPO, parameterName=Intensity}
  )
  StaticSwitchParameter(True=Multiply(A=@n56, B=Transform(None=@n67)), False=Constant()) {group=WPO, parameterName=WPO}
  TextureCoordinate() {uTiling=1.25, vTiling=-0.125}
 ]
}
